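# Sound cue for the ready sound of the WWII cosmetic pickaxe, classed as melee weapon audio.
# It plays one pickaxe swing wave through a Modulator, reused here with raised pitch and lowered volume on the cue.
# A shared pickaxe-ready property sheet and a close-range projectile attenuation asset are assigned.
# It primes on load and overrides concurrency to block new instances once the limit is reached.
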SoundCue /BRCosmetics/Sounds/PickAxes/WWII/PickAxe_Ready_WWII_Cue
root=Modulator_0
nodes [
 {id=Modulator_0, children [WavePlayer_0]}
 {id=WavePlayer_0, properties {soundWaveAssetPtr=&/Pickaxe_WWII_Swing_03}}
]
properties {
 attenuationSettings=&Fort_Weapon_Projectile_Close_Athena_Att
 audioPropertiesSheet=/BRCosmetics/Sounds/Settings/PropertySheets/Pickaxe/APS_Cosmetic_Pickaxe_Ready
 bOverrideConcurrency=true
 bPrimeOnLoad=true
 maxDistance=4950
 pitchMultiplier=1.2
 soundClassObject=&Fort_WeaponMelee
 volumeMultiplier=0.2
 concurrencyOverrides {resolutionRule=PreventNew}
}
